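# Translucent, unlit, two-sided wing effect material for the CoatCheck glider, usable on skeletal and Niagara meshes.
# Color blends desaturatable rainbow texture, curve-atlas gradients and panning noise, with time-of-day tint and optional pattern overlay.
# Opacity lerps between backpack-texture glow and noise-distorted procedural UV masks, with optional tail/top fades and depth fade skipped on forward and mobile.
# Vertex color drives camera offset; dynamic parameters expose wave motion and brightness.
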Material /Game/Gadgets/Assets/VinderTech_GliderChute/Glider_CoatCheck/FX/M_Glider_CoatCheck_Wing
graph {
 properties {
  bUsedWithMeshDeformer=true
  bUsedWithNiagaraMeshParticles=true
  bUsedWithSkeletalMesh=true
  blendMode=BLEND_Translucent
  pixelDepthOffsetMode=PDOM_Legacy
  shadingModel=MSM_Unlit
  twoSided=true
  usedShadingModels=Unlit
 }
 nodes [
  n1 = Add(A=Constant() {r=-3}, B=Multiply(A=VertexColor().G) {constB=-4})
  n2 = Multiply(A=VertexColor().B) {constB=-4}
  LinearInterpolate(
   B=MaterialFunctionCall(Offset Amount=Add(A=@n1, B=@n2)) {materialFunction=/Engine/Functions/Engine_MaterialFunctions02/WorldPositionOffset/CameraOffset}.Camera Offset
   Alpha=ScalarParameter() {defaultValue=1, parameterName=CameraOffset Intensity}
  )
  n3 = TextureCoordinate()
  n4 = Constant2Vector() {g=0.1}
  n5 = Power(Base=ComponentMask(None=Add(A=@n3, B=@n4)) {g=true}) {constExponent=1.5}
  LinearInterpolate(Alpha=@n5) {constA=0.05}
  n6 = Constant() {r=1}
  n7 = VectorParameter() {parameterName="Color_Night (V3)", defaultValue=<rgba 0.7096 0.6208 1 1>}
  n8 = ScalarParameter() {defaultValue=0.6, parameterName=NoiseLerpMin}
  n9 = DynamicParameter() {defaultValue=<rgba 0.7 1 0 -1>, paramNames [Wave_TilingU,Wave_TilingV,Wave_SpeedX,Wave_SpeedY]}
  n10 = Multiply(A=TextureCoordinate(), B=AppendVector(A=@n9.Wave_TilingU, B=@n9.Wave_TilingU))
  n11 = AppendVector(A=@n9.Wave_SpeedX, B=@n9.Wave_SpeedX)
  n12 = Add(A=TextureCoordinate() {uTiling=0.7}, B=AppendVector(A=Constant(), B=VertexColor().R))
  n13 = TextureSample(UVs=LinearInterpolate(
    A=Panner(Coordinate=@n10, Speed=@n11) {speedY=-1}
    B=Panner(Coordinate=@n12) {speedY=-1}
    Alpha=ScalarParameter() {defaultValue=1, parameterName=Use_WaveDynamic}
   )) {samplerType=SAMPLERTYPE_Grayscale, texture=&LowResBlurredNoise}
  n14 = Panner(
   Coordinate=Add(A=TextureCoordinate() {uTiling=0.3, vTiling=0.2}, B=Constant2Vector() {r=0.1})
   Speed=AppendVector(A=ScalarParameter() {parameterName=ColorWave_SpeedX}, B=ScalarParameter() {parameterName=ColorWave_SpeedY})
  ) {speedY=-0.1}
  n15 = Multiply(A=Desaturation(
    None=TextureSample(UVs=@n14) {texture=&T_Tile_RainbowColor}.RGB
    Fraction=ScalarParameter() {defaultValue=1, parameterName=ColorDes}
   ))
  n16 = Constant() {r=0.7}
  n17 = CurveAtlasRowParameter(CurveTime=ComponentMask(None=TextureCoordinate()) {g=true}) {
   atlas=&/CAtlas_Glider_CoatCheck
   curve=&/CV_Glider_CoatCheck_ColorLerp
   defaultValue=1
   parameterName=ColorLerp
  }
  n18 = TextureCoordinate()
  n19 = Power(Base=Sine(None=ComponentMask(None=@n18) {r=true}) {period=2}) {constExponent=8}
  n20 = CurveAtlasRowParameter(CurveTime=ComponentMask(None=@n18) {g=true}) {
   atlas=&CAtlas_Glider_CoatCheck
   curve=&/CV_Glider_CoatCheck
   parameterName=MaskRange
  }
  n21 = LinearInterpolate(Alpha=Multiply(A=@n20.G)) {constA=-0.3}
  n22 = Multiply(A=Add(A=Subtract(A=@n19, B=@n21) {constB=0}, B=NamedRerouteUsage())) {constB=3}
  n23 = Multiply(A=Add(A=Saturate(None=@n22)) {constB=0})
  n24 = Subtract(
   A=Multiply(A=@n17) {constB=1.5}
   B=LinearInterpolate(B=Multiply(A=@n23) {constB=0.4}, Alpha=ScalarParameter() {defaultValue=1, parameterName=Add CenterDetail})
  )
  n25 = Add(B=Multiply(A=Multiply(B=@n24) {constA=1}, B=LinearInterpolate(B=@n15, Alpha=@n16) {constA=1}))
  n26 = Multiply(
   A=Multiply(A=LinearInterpolate(A=@n8, Alpha=@n13.R) {constA=0.4, constB=1.5}, B=LinearInterpolate(A=@n15, B=@n25, Alpha=@n16))
   B=LinearInterpolate(B=ParticleColor().RGB, Alpha=ScalarParameter() {parameterName=UseParticleColor}) {constA=1}
  )
  n27 = ScalarParameter() {defaultValue=1, parameterName=MainEmissive}
  n28 = Desaturation(
   None=Multiply(
    A=MaterialFunctionCall(Color_Morning=@n6, Color_Day=@n6, Color_Evening=@n6, Color_Night=@n7.RGB) {materialFunction=&Color_Time_of_Day}.Result
    B=Multiply(A=@n26, B=@n27)
   )
   Fraction=ScalarParameter() {parameterName=Desaturation}
  )
  n29 = Multiply(A=@n28, B=ScalarParameter() {defaultValue=0.5, group=Pattern, parameterName=Pattern Dark, sliderMax=2}) {constB=0.5}
  n30 = TextureCoordinate()
  n31 = AppendVector(
   A=ScalarParameter() {defaultValue=1, group=Pattern, parameterName=Pattern Tile X, sliderMax=2}
   B=ScalarParameter() {defaultValue=1, group=Pattern, parameterName=Pattern Tile Y, sliderMax=2}
  )
  n32 = ScalarParameter() {group=Pattern, parameterName=Pattern Speed X, sliderMax=2, sliderMin=-2}
  n33 = ScalarParameter() {group=Pattern, parameterName=Pattern Speed Y, sliderMax=2, sliderMin=-2}
  n34 = TextureSampleParameter2D(UVs=Panner(Coordinate=Multiply(A=@n30, B=@n31), Speed=AppendVector(A=@n32, B=@n33)) {speedY=-0.05}) {
   group=Pattern
   parameterName=Pattern Texture
   samplerType=SAMPLERTYPE_LinearGrayscale
   texture=/Game/Effects/Fort_Effects/Emotes/UltraLight/Texture/T_Plasma_A_B_Greyscale
  }
  MaterialFunctionCall(Emissive=StaticSwitchParameter(True=LinearInterpolate(A=@n29, B=@n28, Alpha=@n34.R), False=@n28) {group=Pattern, parameterName=AddPatternOverlay}) {materialFunction=&MF_FX_EmissiveAdjustments}
  n35 = Add(A=TextureCoordinate(), B=NamedRerouteUsage())
  n36 = Subtract(A=OneMinus(None=ComponentMask(None=@n35) {g=true}), B=ScalarParameter() {defaultValue=0.2, group=Mask, parameterName=BottomMaskOffset}) {constB=0.2}
  NamedRerouteDeclaration(None=Saturate(None=Multiply(A=@n36) {constB=10})) {name=Bottom mask, nodeColor=<rgba 0.1419 0 1 1>}
  n37 = TextureCoordinate()
  n38 = DynamicParameter() {parameterIndex=2, defaultValue=<rgba 2 1 0 -1>, paramNames [WaveMask_TilingU, WaveMask_TilingV, WaveMask_SpeedX, WaveMask_SpeedY]}
  n39 = AppendVector(A=@n38.WaveMask_TilingU, B=@n38.WaveMask_TilingU)
  n40 = LinearInterpolate(
   A=Panner(Coordinate=Multiply(A=@n37, B=@n39), Speed=AppendVector(A=@n38.WaveMask_SpeedX, B=@n38.WaveMask_SpeedX)) {speedY=-1}
   B=Panner(Coordinate=TextureCoordinate() {uTiling=2}) {speedY=-1}
   Alpha=ScalarParameter() {defaultValue=1, parameterName=Use_WaveMaskDynamic}
  )
  NamedRerouteDeclaration(None=Multiply(
    A=TextureSample(UVs=@n40) {samplerType=SAMPLERTYPE_Grayscale, texture=&LowResBlurredNoise}.R
    B=ScalarParameter() {defaultValue=0.3, parameterName=Noise Intensity}
   ) {constB=0.3}) {name=Disturb, nodeColor=<rgba 1 0.0752 0 1>}
  n41 = Add(A=Rotator(Coordinate=TextureCoordinate(), Time=ScalarParameter() {defaultValue=-1.5, parameterName=MaskRotator}), B=NamedRerouteUsage())
  n42 = Subtract(A=OneMinus(None=ComponentMask(None=@n41) {g=true}), B=ScalarParameter() {defaultValue=0.6, group=Mask, parameterName=BottomMaskOffset02}) {constB=0.2}
  NamedRerouteDeclaration(None=Saturate(None=Multiply(A=@n42) {constB=6})) {name=Wave_mask, nodeColor=<rgba 1 0.3058 0.6043 1>}
  n43 = LinearInterpolate(B=ParticleColor().A, Alpha=ScalarParameter() {parameterName=UseParticleColor}) {constA=1}
  n44 = TextureSample() {texture=/Game/Accessories/FORT_Backpacks/Backpack_F_MED_CoatCheck/FX/T_Backpack_CoatCheck_mask}
  n45 = NamedRerouteUsage()
  n46 = DynamicParameter() {parameterIndex=3, defaultValue=<rgba 0.25 8 1 1>, paramNames [Wing_Bright, WingGlow_Bright, Param3, Param4]}
  n47 = Saturate(None=Multiply(A=Multiply(A=@n44.R, B=@n45), B=@n46.Wing_Bright) {constB=0.25})
  n48 = TextureCoordinate()
  n49 = ComponentMask(None=@n48) {r=true}
  n50 = Multiply(A=OneMinus(None=@n49)) {constB=2}
  n51 = Multiply(A=@n49) {constB=5}
  n52 = Multiply(A=Multiply(A=Saturate(None=@n50), B=Saturate(None=@n51)), B=OneMinus(None=ComponentMask(None=@n48) {g=true}))
  n53 = OneMinus(None=SphereMask(A=TextureCoordinate() {uTiling=0.6, vTiling=0.6}, B=Constant2Vector() {g=0.7, r=-0.1}, Radius=Constant() {r=5}, Hardness=Constant() {r=0.08}))
  n54 = Multiply(A=Multiply(A=@n45, B=Multiply(A=@n52, B=@n53)), B=@n46.WingGlow_Bright) {constB=8}
  n55 = TextureCoordinate()
  n56 = Power(Base=OneMinus(None=ComponentMask(None=@n55) {g=true}), Exp=ScalarParameter() {defaultValue=2, group=Mask, parameterName=TailFadePower})
  n57 = ScalarParameter() {group=Mask, parameterName=Use TailFade}
  n58 = TextureCoordinate()
  n59 = Power(Base=Sine(None=ComponentMask(None=@n58) {r=true}) {period=2}) {constExponent=1.5}
  n60 = CurveAtlasRowParameter(CurveTime=ComponentMask(None=@n58) {g=true}) {
   atlas=&CAtlas_Glider_CoatCheck
   curve=&CV_Glider_CoatCheck
   parameterName=MaskRange
  }
  n61 = LinearInterpolate(
   A=ScalarParameter() {defaultValue=-1, group=Mask, parameterName=MaskLerpMin}
   Alpha=Multiply(A=StaticComponentMaskParameter(None=@n60) {defaultR=true, parameterName=Mask})
  ) {constA=-1}
  n62 = Multiply(A=Add(A=Subtract(A=@n59, B=@n61) {constB=0}, B=NamedRerouteUsage()), B=ScalarParameter() {defaultValue=5, parameterName=MaskIntensity}) {constB=5}
  n63 = Sine(None=ComponentMask(None=TextureCoordinate()) {r=true}) {period=2}
  n64 = Multiply(A=Multiply(A=Saturate(None=@n62), B=NamedRerouteUsage()), B=Saturate(None=Multiply(A=@n63) {constB=2}))
  n65 = TextureCoordinate()
  n66 = Saturate(None=Multiply(A=ComponentMask(None=@n65) {g=true}) {constB=20})
  n67 = LinearInterpolate(
   A=LinearInterpolate(A=Multiply(A=@n5, B=@n47), B=Saturate(None=@n54), Alpha=ScalarParameter() {defaultValue=1, parameterName=Use_BackpackMaskTexture})
   B=Multiply(A=Multiply(A=@n64, B=@n66), B=LinearInterpolate(B=@n56, Alpha=@n57) {constA=1})
   Alpha=ScalarParameter() {defaultValue=1, parameterName=Use_BackpackMask}
  )
  n68 = Add(A=TextureCoordinate(), B=AppendVector(A=ScalarParameter() {parameterName=Top mask UAdd}, B=ScalarParameter() {parameterName=Top mask VAdd}))
  n69 = Saturate(None=Multiply(A=ComponentMask(None=@n68) {g=true}) {constB=3})
  n70 = ScalarParameter() {parameterName=Use Top mask}
  n71 = Multiply(
   A=Multiply(A=Multiply(A=@n43, B=@n67), B=LinearInterpolate(B=@n69, Alpha=@n70) {constA=1})
   B=ScalarParameter() {defaultValue=1, parameterName=MainAlpha}
  )
  n72 = DepthFade(Opacity=@n71, FadeDistance=ScalarParameter() {defaultValue=2, parameterName=DepthFade})
  ShadingPathSwitch(Default=@n72, Deferred=@n72, Forward=@n71, Mobile=@n71)
 ]
}
